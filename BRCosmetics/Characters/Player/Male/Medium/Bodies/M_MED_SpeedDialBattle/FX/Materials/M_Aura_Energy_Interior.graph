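# Unlit translucent Niagara sprite and mesh particle material for the energy aura interior effect on the SpeedDialBattle player body.
# Emissive sums a gradient mask panned by GlowPan along Z or a local vector, caustic noise hue-shifting particle color, and Fresnel from flattened SceneTexture normals.
# Shading-path switches substitute simpler Fresnel values on forward and mobile.
# Opacity combines mask and sphere-masked Fresnel with opposed depth fades, limiting visibility to near scene depth.
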Material /BRCosmetics/Characters/Player/Male/Medium/Bodies/M_MED_SpeedDialBattle/FX/Materials/M_Aura_Energy_Interior
graph {
 properties {
  bUsedWithNiagaraMeshParticles=true
  bUsedWithNiagaraSprites=true
  blendMode=BLEND_Translucent
  pixelDepthOffsetMode=PDOM_Legacy
  shadingModel=MSM_Unlit
  usedShadingModels=Unlit
 }
 nodes [
  Constant() {r=1}
  n1 = Divide(A=Subtract(A=ObjectPositionWS(), B=WorldPosition().XYZ), B=ScalarParameter() {defaultValue=100, parameterName=Mask_UV}) {constB=100}
  n2 = VectorParameter() {desc=Which Direction you would like the Mask to pan, group=Local Space Direction, parameterName=ObjectsVector, defaultValue=<rgba 0 1 0 1>}
  n3 = Transform(None=@n2.RGB) {transformSourceType=TRANSFORMSOURCE_Local}
  n4 = LinearInterpolate(
   A=ComponentMask(None=@n1) {b=true}
   B=DotProduct(A=Normalize(VectorInput=@n3), B=@n1)
   Alpha=ScalarParameter() {group=Local Space Direction, parameterName=UseLocalSpaceDirection, sliderMax=1}
  )
  n5 = DynamicParameter() {defaultValue=<rgba 3 1 1 1>, paramNames [FresnelExponent, Refraction, GlowPan, Param4]}
  n6 = ComponentMask(None=TextureCoordinate()) {g=true}
  n7 = Multiply(
   A=TextureSampleParameter2D(UVs=Panner(Coordinate=@n4, Time=@n5.GlowPan) {speedY=3}) {parameterName=Mask_Texture, texture=/Game/Global/Textures/Default/Masks_and_Gradients/T_Vert_Gradient_Soft}.R
   B=SmoothStep(Value=OneMinus(None=@n6)) {constMax=0.2}
  ) {constB=1.25}
  n8 = Multiply(
   A=TextureCoordinate()
   B=AppendVector(A=ScalarParameter() {defaultValue=1.5, parameterName=AlphaScaleX}, B=ScalarParameter() {defaultValue=0.5, parameterName=AlphaScaleY})
  )
  n9 = ParticleRandom()
  n10 = TextureCoordinate()
  n11 = TextureSample!(UVs=Panner(Coordinate=Add(A=@n8, B=@n9)) {speedY=0.75}, Tex, "DDX(UVs)"=DDX(Value=@n10), "DDY(UVs)"=DDY(Value=@n10), Apply View MipBias) {
   mipValueMode=TMVM_Derivative
   samplerType=SAMPLERTYPE_LinearGrayscale
   texture=/Game/Global/Textures/Default/Caustics/T_Nosie_Caustics_Grayscale
  }
  n12 = Add(A=Multiply(A=@n11.R, B=ScalarParameter() {defaultValue=2, parameterName=Param_1}) {constB=10}) {constB=0.25}
  n13 = ParticleColor()
  n14 = MaterialFunctionCall(Hue Shift Percentage=Multiply(A=Saturate(None=@n12), B=ScalarParameter() {defaultValue=0.25, parameterName=HueShift}), Texture=@n13.RGB) {materialFunction=&HueShift}
  n15 = Multiply(A=@n12, B=@n14.Result)
  n16 = SceneTexture() {sceneTextureId=PPI_WorldNormal}
  n17 = Fresnel(
   ExponentIn=DynamicParameter() {defaultValue=<rgba 3 1 1 1>, paramNames [FresnelExponent, Refraction, GlowPan, Param4]}.FresnelExponent
   Normal=MaterialFunctionCall(Normal=ComponentMask(None=@n16.Color) {b=true, g=true, r=true}, Flatness=ScalarParameter() {defaultValue=0.15, parameterName=FlattenNormal}) {materialFunction=&FlattenNormal}.Result
  ) {baseReflectFraction=0, exponent=1}
  n18 = ShadingPathSwitch(Default=@n17, Deferred=@n17, Forward=VectorParameter() {parameterName=Forward, defaultValue=<rgba 0 0 1 1>}.RGB, Mobile=@n17)
  n19 = ShadingPathSwitch(Default=@n18, Deferred=@n18, Mobile=Constant() {r=0.05}) {desc=Migrated from FeatureLevelSwitch}
  n20 = Multiply(A=@n14.Result, B=@n19)
  MaterialFunctionCall(Emissive=Add(A=@n7, B=Add(A=@n15, B=@n20))) {materialFunction=&MF_FX_EmissiveAdjustments}
  n21 = TextureCoordinate()
  n22 = Constant() {r=0.5}
  n23 = ScalarParameter() {defaultValue=0.5, parameterName=SphereMaskRadius}
  n24 = ScalarParameter() {parameterName=SphereMaskHardness}
  n25 = Add(A=@n7, B=Multiply(A=@n19, B=SphereMask(A=@n21, B=@n22, Radius=@n23, Hardness=@n24) {hardnessPercent=0}))
  n26 = ScalarParameter() {defaultValue=15, parameterName=VisibleDepthDistance}
  n27 = DepthFade(
   Opacity=OneMinus(None=DepthFade(FadeDistance=@n26) {fadeDistanceDefault=80})
   FadeDistance=ScalarParameter() {defaultValue=15, parameterName=DepthFade}
  ) {fadeDistanceDefault=15}
  Multiply(A=Multiply(A=@n13.A, B=Multiply(A=@n25, B=@n27)), B=ScalarParameter() {defaultValue=1, group=Blinking, parameterName=Opacity})
  ScalarParameter() {defaultValue=3}
  ScalarParameter() {defaultValue=1, parameterName=GlowPanScale}
 ]
}
