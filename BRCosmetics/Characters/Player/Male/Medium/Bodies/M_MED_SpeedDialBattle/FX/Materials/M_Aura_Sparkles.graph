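# Additive unlit Niagara sprite and mesh particle material for the SpeedDialBattle character's sparkle aura.
# Three bokeh-dot layers (detail, base, macro) pan with per-particle random and dynamic UV offsets, then sum.
# A scalar blends between gradient and spherical mask subtraction, with dynamic dissolve only on the gradient branch.
# Particle color tints the result before HDR remapping; particle alpha drives opacity, depth-faded on default and deferred paths.
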Material /BRCosmetics/Characters/Player/Male/Medium/Bodies/M_MED_SpeedDialBattle/FX/Materials/M_Aura_Sparkles
graph {
 properties {
  bAutomaticallySetUsageInEditor=false
  bUseTranslucencyVertexFog=false
  bUsedWithNiagaraMeshParticles=true
  bUsedWithNiagaraSprites=true
  blendMode=BLEND_Additive
  shadingModel=MSM_Unlit
  usedShadingModels=Unlit
 }
 nodes [
  n1 = Multiply(
   A=AppendVector(A=ScalarParameter() {defaultValue=1, parameterName=X Scale Mult}, B=ScalarParameter() {defaultValue=1, parameterName=Y Scale Mult})
   B=Add(A=TextureCoordinate(), B=ParticleRandom())
  )
  n2 = VectorParameter() {parameterName=Detail Texture, sortPriority=4, channelNames {a=V Speed, b=U Speed, g=V Scale, r=U Scale}, defaultValue=<rgba 2 2 0.15 1.5>}
  n3 = Multiply(A=@n1, B=AppendVector(A=@n2.U Scale, B=@n2.U Scale))
  n4 = Multiply(A=Time(), B=ScalarParameter() {defaultValue=1, parameterName=Pan Speed Mult, sortPriority=3})
  n5 = AppendVector(A=@n2.U Speed, B=@n2.U Speed)
  n6 = MaterialFunctionCall(X=Constant(), Y=DynamicParameter() {defaultValue=<rgba 0 0 1 1>, paramNames [OffsetTex,Dissolve,Param3,Param4]}.OffsetTex) {materialFunction=&MakeFloat2}
  n7 = TextureObject() {
   samplerType=SAMPLERTYPE_LinearColor
   texture=/BRCosmetics/Characters/Player/Male/Medium/Bodies/M_MED_SpeedDialBattle/FX/Textures/T_Bokeh_Dots
  }
  n8 = TextureSample(UVs=Add(A=Panner(Coordinate=@n3, Time=@n4, Speed=@n5) {speedX=0.3, speedY=-4}, B=@n6.Result), Tex=@n7) {samplerType=SAMPLERTYPE_LinearColor, texture=/Game/Effects/Fort_Effects/Textures/Tile/T_Random_Dots}
  n9 = VectorParameter() {parameterName=Base Texture, sortPriority=5, channelNames {a=V Speed, b=U Speed, g=V Scale, r=U Scale}, defaultValue=<rgba 1 1 -0.1 1.25>}
  n10 = Multiply(A=@n1, B=AppendVector(A=@n9.U Scale, B=@n9.U Scale))
  n11 = AppendVector(A=@n9.U Speed, B=@n9.U Speed)
  n12 = TextureSample(UVs=Add(A=Panner(Coordinate=@n10, Time=@n4, Speed=@n11) {speedX=-0.3, speedY=-3}, B=@n6.Result), Tex=@n7) {samplerType=SAMPLERTYPE_LinearColor, texture=&T_Random_Dots}
  n13 = VectorParameter() {parameterName=Macro Texture, sortPriority=7, channelNames {a=V Speed, b=U Speed, g=V Scale, r=U Scale}, defaultValue=<rgba 1 1 0.1 0.4>}
  n14 = AppendVector(A=@n13.U Scale, B=@n13.U Scale)
  n15 = Add(
   A=Panner(Coordinate=Multiply(A=@n1, B=@n14), Time=@n4, Speed=AppendVector(A=@n13.U Speed, B=@n13.U Speed)) {speedX=0.5, speedY=-1}
   B=@n6.Result
  )
  n16 = TextureCoordinate()
  n17 = Multiply(A=ComponentMask(None=@n16) {g=true}) {constB=2}
  n18 = OneMinus(None=ComponentMask(None=@n16) {g=true})
  n19 = OneMinus(None=Multiply(A=Power(Base=@n17) {constExponent=3}, B=Multiply(A=@n18)))
  n20 = OneMinus(None=SphereMask(A=TextureCoordinate(), B=Constant() {r=0.5}) {attenuationRadius=1, hardnessPercent=0})
  n21 = Subtract(
   A=Add(
    A=Add(A=@n8.R, B=@n12.R)
    B=TextureSample(UVs=@n15, Tex=@n7) {samplerType=SAMPLERTYPE_LinearColor, texture=&T_Random_Dots}.R
   )
   B=LinearInterpolate(A=Multiply(A=@n19) {constB=2}, B=Multiply(A=@n20) {constB=4}, Alpha=ScalarParameter() {defaultValue=1, parameterName=UseSphericalMask?})
  )
  n22 = ComponentMask(None=TextureCoordinate()) {g=true}
  n23 = LinearInterpolate(
   A=Subtract(A=Multiply(A=@n21, B=@n22), B=DynamicParameter() {defaultValue=<rgba 0 0 1 1>, paramNames [OffsetTex,Dissolve,Param3,Param4]}.Dissolve) {constB=0}
   B=@n21
   Alpha=ScalarParameter() {defaultValue=1, parameterName=UseSphericalMask?}
  )
  n24 = Saturate(None=@n23)
  n25 = ParticleColor()
  MaterialFunctionCall(Emissive=MaterialFunctionCall(LDR Input=Multiply(A=@n24, B=@n25.RGB), HDR Tint=@n23) {materialFunction=&DeriveHDRfromLDR}.RemappedRange) {materialFunction=&MF_FX_EmissiveAdjustments}
  n26 = DepthFade(Opacity=@n25.A, FadeDistance=ScalarParameter() {defaultValue=5, parameterName=Depth Fade Distance})
  ShadingPathSwitch(Default=@n26, Deferred=@n26, Forward=@n25.A, Mobile=@n25.A)
 ]
}
