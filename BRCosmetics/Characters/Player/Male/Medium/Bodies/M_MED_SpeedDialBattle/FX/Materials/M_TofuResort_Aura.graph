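# Translucent, unlit, two-sided Niagara material for the SpeedDialBattle character's TofuResort aura effect.
# Panned distortion noise, scrolling and base masks build opacity, with near-camera, depth, optional bottom and view-angle fades; Mobile and Forward skip depth and angle fades.
# A scrolling chromatic texture drives emissive color, tinted by particle color or remapped through an iridescent curve atlas.
# World position offset combines SplineThicken, a Niagara-driven pull and optional sine wobble.
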Material /BRCosmetics/Characters/Player/Male/Medium/Bodies/M_MED_SpeedDialBattle/FX/Materials/M_TofuResort_Aura
graph {
 properties {
  bUseTranslucencyVertexFog=false
  bUsedWithNiagaraMeshParticles=true
  bUsedWithNiagaraRibbons=true
  bUsedWithNiagaraSprites=true
  blendMode=BLEND_Translucent
  pixelDepthOffsetMode=PDOM_Legacy
  shadingModel=MSM_Unlit
  twoSided=true
  usedShadingModels=Unlit
 }
 nodes [
  n1 = VectorParameter() {group=Noise, parameterName=Noise UV, channelNames {a=Speed V, b=Speed U, g=Tile V, r=Tile U}, defaultValue=<rgba 1 0.25 0.25 0.5>}
  n2 = DynamicParameter() {defaultValue=<rgba 0.2 0 0 1>, paramNames [UVTilingRandom, TimeOffsetRandom, Color Mask offset, DistortionIntensity]}
  n3 = Multiply(
   A=TextureCoordinate()
   B=MaterialFunctionCall(X=@n1.Tile U, Y=Add(A=@n2.UVTilingRandom, B=@n1.Tile V)) {materialFunction=&MakeFloat2}.Result
  )
  n4 = Add(A=@n2.TimeOffsetRandom, B=Time())
  n5 = Multiply(
   A=MaterialFunctionCall(X=@n1.Speed U, Y=@n1.Speed U) {materialFunction=&MakeFloat2}.Result
   B=LinearInterpolate(Alpha=ParticleRandom()) {constA=0.75, constB=1.25}
  )
  n6 = Constant()
  n7 = DynamicParameter() {parameterIndex=2, defaultValue=<rgba -60 -60 0 1>, paramNames [WidthBase, WidthTip, Pan V, MaskControl]}
  n8 = TextureSampleParameter2D(UVs=Add(A=Panner(Coordinate=@n3, Time=@n4, Speed=@n5), B=AppendVector(A=@n6, B=@n7.Pan V))) {
   group=Noise
   parameterName=Noise Texture
   samplerType=SAMPLERTYPE_Normal
   texture=&T_Weap_Distort_N_02
  }
  n9 = NamedRerouteUsage()
  n10 = LinearInterpolate(Alpha=ParticleRandom()) {constA=0.75, constB=1.25}
  NamedRerouteDeclaration(None=Multiply(A=ComponentMask(None=@n8.RGB) {desc=Distort mask V2, g=true, r=true}, B=Multiply(A=@n9, B=@n10))) {name=Normal_Noise, nodeColor=<rgba 0.8047 1 0 1>}
  n11 = DynamicParameter() {desc=Parameters set by Niagara, parameterIndex=3, defaultValue=<rgba 0 1 1 1>, paramNames [X,Y,Z,Intensity]}
  n12 = MaterialFunctionCall(Float3=Subtract(
    A=MaterialFunctionCall(X=@n11.X, Y=@n11.X, Z=@n11.X) {materialFunction=&MakeFloat3}.Result
    B=ParticlePositionWS()
   )) {materialFunction=&BreakOutFloat3Components}
  n13 = ScalarParameter() {defaultValue=100, desc=Interpolate Clamp XY, group=WPO, parameterName=Interpolate XY Clamp, sortPriority=3}
  n14 = ScalarParameter() {defaultValue=2, desc=Interpolate Clamp Z, group=WPO, parameterName=Interpolate Clamp Z, sortPriority=4}
  n15 = AppendVector(
   A=Clamp(None=AppendVector(A=@n12.R, B=@n12.R), Min=Multiply(A=@n13) {constB=-1}, Max=@n13) {maxDefault=1000, minDefault=-1000}
   B=Clamp(None=@n12.B, Min=Multiply(A=@n14) {constB=-1}, Max=@n14) {minDefault=-1}
  )
  n16 = OneMinus(None=TextureCoordinate())
  n17 = DynamicParameter() {parameterIndex=1, defaultValue=<rgba 1 1 5 0>, paramNames [Dive Scroll Speed, AngleFade, Interpolate Power, ColorCurve]}
  n18 = Multiply(A=@n11.Intensity, B=Power(Base=ComponentMask(None=@n16) {g=true}, Exp=@n17.Interpolate Power))
  n19 = DynamicParameter() {parameterIndex=2, defaultValue=<rgba -60 -60 0 1>, paramNames [WidthBase, WidthTip, Pan V, MaskControl]}
  n20 = ScalarParameter() {defaultValue=1, parameterName=WPO_Width}
  n21 = Multiply(A=@n19.WidthBase, B=@n20)
  n22 = VertexColor()
  n23 = Multiply(A=@n19.WidthTip, B=@n20)
  n24 = Add(
   A=StaticSwitchParameter(True=LinearInterpolate(B=@n15, Alpha=@n18), False=Constant()) {defaultValue=true, group=WPO, parameterName=Use Interpolate WPO, sortPriority=1}
   B=MaterialFunctionCall(WidthBase=Multiply(A=@n21, B=@n22.R), WidthTip=Multiply(A=@n23, B=@n22.B)) {materialFunction=&SplineThicken}.WorldPositionOffset
  )
  n25 = TextureCoordinate()
  n26 = Add(A=Time(), B=ParticleRandom())
  n27 = LinearInterpolate(Alpha=ParticleRandom()) {constA=0.7, constB=1.3}
  n28 = Multiply(
   A=Add(A=ComponentMask(None=@n25) {g=true}, B=Multiply(A=@n26, B=@n27))
   B=ScalarParameter() {defaultValue=0.9, group=WPO, parameterName=WPOSineAmount, sortPriority=10}
  )
  n29 = Sine(None=@n28)
  n30 = Sine(None=@n28) {period=1.33}
  n31 = Multiply(A=Add(A=Add(A=@n29, B=@n30), B=Sine(None=@n28) {period=1.66}), B=Constant3Vector() {constant=<rgba 0.75 0.75 0 1>})
  n32 = Multiply(
   A=ScalarParameter() {defaultValue=15, group=WPO, parameterName=WPOSineIntensity, sortPriority=11}
   B=LinearInterpolate(Alpha=ParticleRandom()) {constA=0.6}
  )
  n33 = ComponentMask(None=TextureCoordinate()) {g=true}
  n34 = Add(A=@n24, B=Multiply(A=Multiply(A=@n31, B=@n32), B=OneMinus(None=@n33)))
  NamedRerouteDeclaration(None=LinearInterpolate(
    B=StaticSwitchParameter(True=@n34, False=@n24) {group=WPO, parameterName=Use Wobble WPO, sortPriority=9}
    Alpha=ScalarParameter() {defaultValue=1, parameterName=UseWPO?}
   )) {name=WPO, nodeColor=<rgba 0 0 0 1>}
  n35 = TextureCoordinate()
  n36 = ComponentMask(None=@n35) {g=true}
  n37 = OneMinus(None=ComponentMask(None=@n35) {r=true})
  n38 = ScalarParameter() {parameterName=HorizontalMask?}
  n39 = Add(
   A=LinearInterpolate(
    A=TextureCoordinate() {uTiling=0.75}
    B=MaterialFunctionCall(X=@n36, Y=@n37) {materialFunction=&MakeFloat2}.Result
    Alpha=@n38
   )
   B=LinearInterpolate(A=Constant2Vector() {r=0.125}, B=Constant2Vector(), Alpha=@n38)
  )
  n40 = TextureCoordinate()
  n41 = OneMinus(None=ComponentMask(None=@n40) {g=true})
  n42 = OneMinus(None=ComponentMask(None=@n40) {r=true})
  n43 = ScalarParameter() {parameterName=HorizontalMask?}
  n44 = Saturate(None=Add(
    A=TextureSample(UVs=@n39) {samplerSource=SSM_Clamp_WorldGroupSettings, texture=/Game/WaterAndWind/Textures/WeaponFireVectorMask}.R
    B=LinearInterpolate(A=@n41, B=@n42, Alpha=@n43)
   ))
  n45 = Multiply(A=@n44, B=@n44) {constB=2}
  NamedRerouteDeclaration(None=Saturate(None=Multiply(A=@n45) {constB=3})) {name=Bottom Fade Mask, nodeColor=<rgba 0.1484 0 1 1>}
  n46 = VectorParameter() {parameterName=Base, channelNames {a=NoiseStrength, b=MaskMultiply, g=DepthFade, r=""}, defaultValue=<rgba 0 50 1 0>}
  NamedRerouteDeclaration(None=@n46.DepthFade) {name=DepthFade_Param, nodeColor=<rgba 1 0.3516 0 1>}
  NamedRerouteDeclaration(None=@n46.MaskMultiply) {name=MaskMultiply_Param, nodeColor=<rgba 1 0 0.0469 1>}
  n47 = DynamicParameter() {defaultValue=<rgba 0.2 0 0 1>, paramNames [UVTilingRandom, TimeOffsetRandom, Color Mask offset, DistortionIntensity]}
  NamedRerouteDeclaration(None=Multiply(A=@n46.NoiseStrength, B=@n47.DistortionIntensity)) {name=NoiseStrength_Param, nodeColor=<rgba 1 0 0.2578 1>}
  n48 = TextureCoordinate()
  n49 = LinearInterpolate(
   A=ComponentMask(None=@n48) {g=true}
   B=OneMinus(None=ComponentMask(None=@n48) {r=true})
   Alpha=ScalarParameter() {parameterName=HorizontalMask?}
  )
  n50 = MaterialFunctionCall() {materialFunction=&Swizzle}
  n51 = TextureCoordinate()
  n52 = ScalarParameter() {parameterName=HorizontalMask?}
  n53 = DynamicParameter() {defaultValue=<rgba 0.2 0 0 1>, paramNames [UVTilingRandom, TimeOffsetRandom, Color Mask offset, DistortionIntensity]}
  n54 = VectorParameter() {group=Mask, parameterName=Mask UV, channelNames {a=Speed V, b=Speed U, g=Tile V, r=Tile U}, defaultValue=<rgba 0.5 1 0.5 0>}
  n55 = Add(A=@n53.UVTilingRandom, B=@n54.Tile U)
  n56 = ParticleRandom()
  n57 = Panner(
   Coordinate=Multiply(
    A=LinearInterpolate(A=@n50.YX, B=@n51, Alpha=@n52)
    B=MaterialFunctionCall(X=@n55, Y=@n54.Tile V) {materialFunction=&MakeFloat2}.Result
   )
   Time=Add(A=@n53.TimeOffsetRandom, B=Time())
   Speed=Multiply(
    A=MaterialFunctionCall(X=@n54.Speed U, Y=@n54.Speed U) {materialFunction=&MakeFloat2}.Result
    B=LinearInterpolate(Alpha=@n56) {constA=0.75, constB=1.25}
   )
  )
  n58 = Multiply(A=NamedRerouteUsage(), B=Constant2Vector() {g=0.1, r=1})
  n59 = DynamicParameter() {parameterIndex=2, defaultValue=<rgba -60 -60 0 1>, paramNames [WidthBase, WidthTip, Pan V, MaskControl]}
  n60 = Constant()
  n61 = TextureSampleParameter2D(UVs=Add(A=Add(A=@n58, B=@n57) {desc=Additional Distortion}, B=AppendVector(A=@n59.Pan V, B=@n60))) {
   group=Mask
   parameterName=Mask Texture
   samplerType=SAMPLERTYPE_Grayscale
   texture=/Game/Effects/Fort_Effects/Textures/Smoke/T_Smoke_Trail_Soft_NoClamp
  }
  NamedRerouteDeclaration(None=Multiply(A=Saturate(None=@n49), B=@n61.G) {desc=Feather Scroll Mask}) {name=Scroll Mask, nodeColor=<rgba 1 0 0.3984 1>}
  n62 = TextureCoordinate()
  n63 = VectorParameter() {group=Color, parameterName=Color UV, channelNames {a=Speed V, b=Speed U, g=Tile V, r=Tile U}, defaultValue=<rgba 1 1 0 0.5>}
  n64 = DynamicParameter() {defaultValue=<rgba 0.2 0 0 1>, paramNames [UVTilingRandom, TimeOffsetRandom, Color Mask offset, DistortionIntensity]}
  n65 = MaterialFunctionCall(X=@n63.Tile U, Y=Add(A=@n63.Tile V, B=@n64.UVTilingRandom)) {materialFunction=&MakeFloat2}
  n66 = ParticleRandom()
  n67 = Panner(
   Coordinate=Add(A=Multiply(A=@n62, B=@n65.Result), B=NamedRerouteUsage())
   Time=Add(A=@n64.TimeOffsetRandom, B=Time())
   Speed=Multiply(
    A=MaterialFunctionCall(X=@n63.Speed U, Y=@n63.Speed U) {materialFunction=&MakeFloat2}.Result
    B=LinearInterpolate(Alpha=@n66) {constA=0.75, constB=1.25}
   )
  ) {speedY=0.5}
  n68 = Constant()
  n69 = DynamicParameter() {parameterIndex=2, defaultValue=<rgba -60 -60 0 1>, paramNames [WidthBase, WidthTip, Pan V, MaskControl]}
  n70 = TextureSampleParameter2D(UVs=Add(A=Add(A=@n67, B=@n64.Color Mask offset), B=AppendVector(A=@n68, B=@n69.Pan V))) {
   parameterName=Color Tex
   texture=/BRCosmetics/Characters/Player/Male/Medium/Bodies/M_MED_SpeedDialBattle/FX/Textures/T_WavePattern_Chromatic_Filled_PNG
  }
  n71 = Multiply(A=Desaturation(None=@n70.RGB, Fraction=ScalarParameter() {parameterName=Saturation}), B=ParticleColor().RGB)
  n72 = ComponentMask(None=@n70.RGB) {g=true}
  n73 = VectorParameter() {parameterName=Refine Color Tex, channelNames {a=Lerp Alpha, b=Mult2, g=Mult1, r=Power}, defaultValue=<rgba 0 0 0 1>}
  n74 = DynamicParameter() {parameterIndex=1, defaultValue=<rgba 1 1 5 0>, paramNames [Dive Scroll Speed, AngleFade, Interpolate Power, ColorCurve]}
  n75 = CurveAtlasRowParameter(CurveTime=Add(
    A=MaterialFunctionCall(Texture=@n72, RGB=@n73, A=@n73) {materialFunction=&MF_FX_Refine}.Result
    B=@n74.ColorCurve
   )) {
   atlas=/Game/Weapons/FORT_Melee/Pickaxe_RainbowStraps/FX/Holofoil/CA_Holofoil_Iridescent
   curve=/Game/Athena/LinearColorCurves/Curve_Holofoil_Iridescent
   parameterName=ColorCurveAtlas
  }
  n76 = ScalarParameter() {parameterName=UseColorCurve}
  NamedRerouteDeclaration(None=MaterialFunctionCall(Emissive=LinearInterpolate(A=@n71, B=@n75, Alpha=@n76)) {materialFunction=&MF_FX_EmissiveAdjustments}.Result) {name=Emissive, nodeColor=<rgba 0 0.1094 1 1>}
  n77 = Constant() {r=64}
  n78 = NamedRerouteUsage()
  n79 = TextureCoordinate()
  n80 = TextureCoordinate()
  n81 = MaterialFunctionCall(X=ComponentMask(None=@n80) {g=true}, Y=OneMinus(None=ComponentMask(None=@n80) {r=true})) {materialFunction=&MakeFloat2}
  n82 = ScalarParameter() {parameterName=HorizontalMask?}
  n83 = TextureSampleParameter2D(UVs=Add(A=LinearInterpolate(A=@n79, B=@n81.Result, Alpha=@n82), B=@n58)) {parameterName=Base_mask, texture=/Game/Effects/Fort_Effects/Textures/Masks/T_Square_Gradient_M}
  n84 = DynamicParameter() {parameterIndex=2, defaultValue=<rgba -60 -60 0 1>, paramNames [WidthBase, WidthTip, Pan V, MaskControl]}
  n85 = Power(
   Base=Multiply(A=Multiply(A=@n83.R, B=@n78) {desc=Apply Base Mask}, B=NamedRerouteUsage())
   Exp=Add(A=@n84.MaskControl, B=ScalarParameter() {parameterName=MaskControl_Debug})
  ) {constExponent=1}
  n86 = VertexColor()
  n87 = Multiply(
   A=Multiply(
    A=ParticleColor().A
    B=MaterialFunctionCall(Distance=@n77) {materialFunction=&Near_Camera_Fade}.Result
   )
   B=Saturate(None=Multiply(A=@n85, B=@n86.G))
  ) {desc=Apply Combined Mask}
  n88 = LinearInterpolate(A=@n87, B=Multiply(A=@n87, B=NamedRerouteUsage()), Alpha=ScalarParameter() {parameterName=ButtomFade})
  n89 = DepthFade(Opacity=@n88, FadeDistance=NamedRerouteUsage())
  n90 = DynamicParameter() {parameterIndex=1, defaultValue=<rgba 1 1 5 0>, paramNames [Dive Scroll Speed, AngleFade, Interpolate Power, ColorCurve]}
  n91 = LinearInterpolate(
   A=@n89
   B=Multiply(
    A=@n89
    B=MaterialFunctionCall("FadeSensitivity (optional)"=@n90.AngleFade) {materialFunction=&FadeBasedOnViewAngle}.Result
   )
   Alpha=ScalarParameter() {parameterName=UseAngleFade}
  )
  NamedRerouteDeclaration(None=ShadingPathSwitch(Default=@n91, Deferred=@n91, Forward=@n88, Mobile=@n88)) {name=Opacity, nodeColor=<rgba 0.4297 1 0 1>}
  NamedRerouteUsage()
  NamedRerouteUsage()
  NamedRerouteUsage()
  Power(Base=@n42) {constExponent=1.25}
 ]
}
