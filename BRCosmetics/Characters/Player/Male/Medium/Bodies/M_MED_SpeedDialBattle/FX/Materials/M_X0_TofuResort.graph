Material /BRCosmetics/Characters/Player/Male/Medium/Bodies/M_MED_SpeedDialBattle/FX/Materials/M_X0_TofuResort
graph {
 properties {
  bAutomaticallySetUsageInEditor=false
  bTangentSpaceNormal=false
  bUseTranslucencyVertexFog=false
  bUsedWithNiagaraMeshParticles=true
  bUsedWithNiagaraSprites=true
  blendMode=BLEND_Translucent
  shadingModel=MSM_Unlit
  twoSided=true
  usedShadingModels=Unlit
 }
 nodes [
  n1 = ScalarParameter() {defaultValue=0.2, parameterName=DistortionIntensity}
  n2 = LinearInterpolate(Alpha=ParticleRandom()) {constA=0.75, constB=1.25}
  n3 = MaterialFunctionCall() {materialFunction=&LinearGradient}
  n4 = ParticleRandom()
  n5 = Multiply(
   A=MaterialFunctionCall(X=Add(A=@n3.VGradient, B=@n3.VGradient) {constB=0.5}, Y=@n3.VGradient) {materialFunction=&MakeFloat2}.Result
   B=LinearInterpolate(Alpha=@n4) {constA=0.75}
  )
  n6 = Add(
   A=Multiply(A=NamedRerouteUsage(), B=Multiply(A=@n1, B=@n2))
   B=Add(
    A=Add(A=@n5, B=@n4) {constB=0.5}
    B=DynamicParameter() {defaultValue=<rgba 0.01 0 0 0>, paramNames [Dissolve, ReactiveOffset, DistortionIntensity, ""]}.ReactiveOffset
   )
  )
  n7 = TextureSampleParameter2D(UVs=@n6) {
   parameterName=Gradient Color Tex
   texture=/BRCosmetics/Characters/Player/Male/Medium/Bodies/M_MED_SpeedDialBattle/FX/Textures/T_Gradient_Tofu
  }
  n8 = TextureSampleParameter2D(UVs=@n6) {parameterName=Color Tex, texture=&T_Tile_RainbowColor}
  n9 = ScalarParameter() {defaultValue=0.3, parameterName=Lerp Colors}
  n10 = ParticleColor()
  MaterialFunctionCall(Emissive=Multiply(A=LinearInterpolate(A=@n7.RGB, B=@n8.RGB, Alpha=@n9) {constAlpha=0.3}, B=@n10.RGB)) {materialFunction=&MF_FX_EmissiveAdjustments}
  n11 = NamedRerouteUsage()
  n12 = DynamicParameter() {defaultValue=<rgba 0.01 0 0 0>, paramNames [Dissolve, ReactiveOffset, DistortionIntensity, ""]}
  n13 = TextureSampleParameter2D(UVs=Add(A=TextureCoordinate(), B=Multiply(A=@n11, B=@n12.DistortionIntensity) {constB=0})) {
   parameterName=Alpha Texture
   samplerType=SAMPLERTYPE_LinearGrayscale
   texture=/BRCosmetics/Characters/Player/Male/Medium/Bodies/M_MED_SpeedDialBattle/FX/Textures/T_TofuResort_Icon
  }
  n14 = TextureCoordinate() {uTiling=1.25, vTiling=1.25}
  n15 = ParticleRandom()
  n16 = LinearInterpolate(A=@n14, B=Multiply(A=@n14) {constB=0.75}, Alpha=@n15)
  n17 = TextureSampleParameter2D(UVs=LinearInterpolate(A=Multiply(A=@n14) {constB=0.8}, B=Add(A=@n16, B=@n15), Alpha=ScalarParameter() {defaultValue=1, parameterName=USE_DISSOLVE_PARTICLE_RANDOM})) {
   parameterName=Dissolve Texture
   samplerType=SAMPLERTYPE_LinearGrayscale
   texture=/BRCosmetics/Gadgets/Assets/VinderTech_GliderChute/Glider_GateHound/FX/T_Mojo_Fire_Noise
  }
  n18 = Saturate(None=Multiply(A=@n13.G, B=@n17.G))
  n19 = DynamicParameter() {defaultValue=<rgba 0.01 0 0 0>, paramNames [Dissolve, ReactiveOffset, DistortionIntensity, ""]}
  n20 = ScalarParameter() {parameterName=Dissolve Softness}
  Multiply(
   A=@n10.A
   B=Saturate(None=MaterialFunctionCall!(Texture=@n18, Erosion=@n19.Dissolve, Softness=@n20) {materialFunction=/Game/Effects/Niagara/MasterMaterials/Functions/MF_Erosion_Smoothstep_01}.Result)
  )
  n21 = TextureCoordinate()
  n22 = VectorParameter() {group=Noise, parameterName=Noise UV, channelNames {a=Speed V, b=Speed U, g=Tile V, r=Tile U}, defaultValue=<rgba 1 2 0.25 0.25>}
  n23 = DynamicParameter() {parameterIndex=1, defaultValue=<rgba 0.2 0 0 1>, paramNames [UVTilingRandom, TimeOffsetRandom, "", ""]}
  n24 = MaterialFunctionCall(X=@n22.Tile U, Y=Add(A=@n23.UVTilingRandom, B=@n22.Tile V)) {materialFunction=&MakeFloat2}
  n25 = Time()
  n26 = MaterialFunctionCall(X=@n22.Speed U, Y=@n22.Speed U) {materialFunction=&MakeFloat2}
  n27 = LinearInterpolate(Alpha=ParticleRandom()) {constA=0.75, constB=1.25}
  n28 = TextureSampleParameter2D(UVs=Panner(Coordinate=Multiply(A=@n21, B=@n24.Result), Time=Add(A=@n23.TimeOffsetRandom, B=@n25), Speed=Multiply(A=@n26.Result, B=@n27))) {
   group=Noise
   parameterName=Distortion Texture
   samplerType=SAMPLERTYPE_Normal
   texture=&T_Weap_Distort_N_02
  }
  NamedRerouteDeclaration(None=ComponentMask(None=@n28.RGB) {desc=Distort mask V2, g=true, r=true}) {name=UV Distortion, nodeColor=<rgba 0.1484 0 1 1>}
 ]
}
