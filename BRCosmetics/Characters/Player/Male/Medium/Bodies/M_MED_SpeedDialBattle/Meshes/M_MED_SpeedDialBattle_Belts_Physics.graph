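# Physics asset for the belts mesh of the male medium SpeedDialBattle player body, defining collision and joint limits for the dangling belt bones.
# Small capsules on two three-bone belt chains hang from the pelvis with tight swing and twist limits.
# Separate left and right belt joints attach to spine_01 with twist locked.
# Larger capsules on the pelvis, lower spine and both thighs provide body volumes alongside the belts.
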PhysicsAsset /BRCosmetics/Characters/Player/Male/Medium/Bodies/M_MED_SpeedDialBattle/Meshes/M_MED_SpeedDialBattle_Belts_Physics
bodies {
 Belts_l_jnt_01 [{shapeType=Capsule, length=3, radius=2, shapeName=Belts_l_jnt_01_capsule, center=<1 0 0>, rotation=<rot 90 0 0>}]
 Belts_r_jnt_01 [{shapeType=Capsule, length=2.7968, radius=2.2032, shapeName=Belts_r_jnt_01_capsule, center=<-1 0 0>, rotation=<rot 90 0 0>}]
 belt_l [{shapeType=Capsule, length=2.1837, radius=1.6453, shapeName=belt_l_capsule, center=<1.1425 0.1139 -0.3474>, rotation=<rot 90 0 0>}]
 belt_l1 [{
   shapeType=Capsule
   length=1.4602
   radius=2.6518
   shapeName=belt_l1_capsule
   center=<0.4847 -0.2512 -0.4049>
   rotation=<rot 86.5081 179.9996 179.9995>
  }]
 belt_l2 [{
   shapeType=Capsule
   length=1.9897
   radius=2.667
   shapeName=belt_l2_capsule
   center=<1.0606 -0.0558 -1.1022>
   rotation=<rot 62.8081 0.0003 0.0002>
  }]
 belt_r [{shapeType=Capsule, length=2.3093, radius=1.1907, shapeName=belt_r_capsule, center=<1.5538 -0.0942 -0.4657>, rotation=<rot 77.1883 0 -0.0001>}]
 belt_r1 [{shapeType=Capsule, length=3.08, radius=1.7939, shapeName=belt_r1_capsule, center=<0.586 0.0253 -0.0793>, rotation=<rot 90 0 0>}]
 belt_r2 [{
   shapeType=Capsule
   length=3.319
   radius=1.7546
   shapeName=belt_r2_capsule
   center=<1.8956 -0.8062 0.7019>
   rotation=<rot 61.5548 -174.4382 -170.9285>
  }]
 pelvis [{
   shapeType=Capsule
   length=5.7214
   radius=11.5296
   shapeName=pelvis_capsule
   center=<1.9773 -0.1996 0.0768>
   rotation=<rot 10.1994 -90.0046 -0.0511>
  }]
 spine_01 [
  {
   shapeType=Capsule
   length=1.6756
   radius=16.0656
   shapeName=spine_01_capsule
   center=<8.3036 -0.2344 0.0704>
   rotation=<rot -90 -0.0004 -0.0002>
  }
  {shapeType=Capsule, length=19.3, radius=6.8773, shapeName=spine_01_capsule, center=<5.7909 -2.1932 0>}
 ]
 thigh_l [{
   shapeType=Capsule
   length=20.4719
   radius=7.8755
   shapeName=thigh_l_capsule
   center=<-11.4099 1.5315 -0.1043>
   rotation=<rot 14.6004 -89.596 -88.454>
  }]
 thigh_r [{
   shapeType=Capsule
   length=20.4719
   radius=7.8755
   shapeName=thigh_l_capsule
   center=<13.1425 -1.4607 0.0407>
   rotation=<rot 14.6557 -99.0994 89.1408>
  }]
}
constraints [
 {bone1Name=belt_l1, bone2Name=belt_l, swing1LimitDegrees=25, swing2LimitDegrees=6, twistLimitDegrees=5}
 {bone1Name=belt_l2, bone2Name=belt_l1, swing1LimitDegrees=15, swing2LimitDegrees=7, twistLimitDegrees=5}
 {bone1Name=belt_r1, bone2Name=belt_r, swing1LimitDegrees=10, swing2LimitDegrees=6, twistLimitDegrees=5}
 {bone1Name=belt_r2, bone2Name=belt_r1, swing1LimitDegrees=15, swing2LimitDegrees=7, twistLimitDegrees=5}
 {bone1Name=belt_l, bone2Name=pelvis, swing1LimitDegrees=10, swing2LimitDegrees=3, twistLimitDegrees=2}
 {bone1Name=belt_r, bone2Name=pelvis, swing1LimitDegrees=5, swing2LimitDegrees=3, twistLimitDegrees=2}
 {bone1Name=Belts_r_jnt_01, bone2Name=spine_01, swing1LimitDegrees=15, swing2LimitDegrees=20, twistLimitDegrees=45, twistMotion=Locked}
 {bone1Name=Belts_l_jnt_01, bone2Name=spine_01, swing1LimitDegrees=15, swing2LimitDegrees=20, twistLimitDegrees=45, twistMotion=Locked}
]
