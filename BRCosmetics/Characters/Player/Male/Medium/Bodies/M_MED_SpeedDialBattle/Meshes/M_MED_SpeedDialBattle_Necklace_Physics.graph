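# Physics asset for the necklace on the medium male SpeedDialBattle player body, set up for dynamic necklace joints against the upper torso.
# Four box bodies form a chain along the dyn_necklace joints, with the first constrained to spine_05.
# Four capsules on spine_05 approximate the torso volume.
# Every link locks twist, the root link also locks one swing axis, and configured swing angles grow toward the last joint.
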PhysicsAsset /BRCosmetics/Characters/Player/Male/Medium/Bodies/M_MED_SpeedDialBattle/Meshes/M_MED_SpeedDialBattle_Necklace_Physics
bodies {
 dyn_necklace_jnt [{shapeType=Box, extentX=5.7041, extentY=3.8569, extentZ=8.1038, shapeName=dyn_necklace_jnt_box}]
 dyn_necklace_jnt2 [{
   shapeType=Box
   extentX=6.4397
   extentY=2.6493
   extentZ=6.9649
   shapeName=dyn_necklace_jnt2_box
   center=<0.006 -0.3213 0.0002>
   rotation=<rot 1.4647 3.4799 0.0891>
  }]
 dyn_necklace_jnt3 [{
   shapeType=Box
   extentX=4.8641
   extentY=1.7693
   extentZ=4.9485
   shapeName=dyn_necklace_jnt3_box
   center=<0.7001 0.2022 0>
   rotation=<rot 0 -1.9604 0>
  }]
 dyn_necklace_jnt4 [{
   shapeType=Box
   extentX=8.2713
   extentY=0.9866
   extentZ=7.984
   shapeName=dyn_necklace_jnt4_box
   center=<4.472 -0.2746 0>
   rotation=<rot 0 -1.6377 0>
  }]
 spine_05 [
  {shapeType=Capsule, length=15.3844, radius=7.5911, shapeName=spine_05_capsule, center=<-7.8005 7.903 -0.1936>}
  {shapeType=Capsule, length=15.0345, radius=10.2782, shapeName=spine_05_capsule, center=<-2.0029 -0.2972 0>}
  {
   shapeType=Capsule
   length=4.7141
   radius=4.3772
   shapeName=spine_05_capsule
   center=<5.2623 9.6395 -8.6062>
   rotation=<rot 74.0539 -27.752 8.2803>
  }
  {
   shapeType=Capsule
   length=4.5779
   radius=4.5134
   shapeName=spine_05_capsule
   center=<5.8871 7.9553 8.6811>
   rotation=<rot 77.9878 151.6859 -168.6626>
  }
 ]
}
constraints [
 {
  bone1Name=dyn_necklace_jnt
  bone2Name=spine_05
  swing1LimitDegrees=3
  swing1Motion=Locked
  swing2LimitDegrees=5
  twistLimitDegrees=45
  twistMotion=Locked
 }
 {
  bone1Name=dyn_necklace_jnt2
  bone2Name=dyn_necklace_jnt
  swing1LimitDegrees=5
  swing2LimitDegrees=10
  twistLimitDegrees=45
  twistMotion=Locked
 }
 {
  bone1Name=dyn_necklace_jnt3
  bone2Name=dyn_necklace_jnt2
  swing1LimitDegrees=5
  swing2LimitDegrees=10
  twistLimitDegrees=45
  twistMotion=Locked
 }
 {
  bone1Name=dyn_necklace_jnt4
  bone2Name=dyn_necklace_jnt3
  swing1LimitDegrees=10
  swing2LimitDegrees=20
  twistLimitDegrees=45
  twistMotion=Locked
 }
]
