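# Physics asset for the M_MED_SpeedDialBattle male medium player body, set up for the motion of its belts.
# Capsules on pelvis, spine_01 and both thighs form the body volumes, with an extra capsule on the right thigh.
# Left and right three-bone belt chains hang from the pelvis with tight angular limits, while two Belts joints attach to spine_01 with twist locked.
# Thigh-to-pelvis rotation is locked and spine-to-pelvis is free.
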PhysicsAsset /BRCosmetics/Characters/Player/Male/Medium/Bodies/M_MED_SpeedDialBattle/Meshes/M_MED_SpeedDialBattle_Physics
bodies {
 Belts_l_jnt_01 [{shapeType=Capsule, length=3, radius=2, shapeName=Belts_l_jnt_01_capsule, center=<1 0 0>, rotation=<rot 90 0 0>}]
 Belts_r_jnt_01 [{shapeType=Capsule, length=3, radius=2, shapeName=Belts_r_jnt_01_capsule, center=<-1 0 0>, rotation=<rot 90 0 0>}]
 belt_l [{shapeType=Capsule, length=3, radius=2, shapeName=belt_l_capsule, center=<-0.0831 0.8588 0.1051>, rotation=<rot -70.0443 0 0.0001>}]
 belt_l1 [{shapeType=Capsule, length=3, radius=2, shapeName=belt_l1_capsule, center=<-0.0857 1.7801 -0.4788>, rotation=<rot 90 0 -8.1375>}]
 belt_l2 [{shapeType=Capsule, length=2.5645, radius=2.966, shapeName=belt_l2_capsule, rotation=<rot 85.6757 0 -0.0001>}]
 belt_r [{
   shapeType=Capsule
   length=4.7912
   radius=1.2088
   shapeName=belt_r_capsule
   center=<-0.0831 0.8588 0.1051>
   rotation=<rot 60.6335 6.1392 -2.8847>
  }]
 belt_r1 [{shapeType=Capsule, length=4, radius=2, shapeName=belt_r1_capsule, rotation=<rot 90 -90 -85.5324>}]
 belt_r2 [{
   shapeType=Capsule
   length=2.9224
   radius=2.6081
   shapeName=belt_r2_capsule
   center=<0.0272 -0.741 -0.1267>
   rotation=<rot 74.3889 -140.2031 -141.2584>
  }]
 pelvis [{shapeType=Capsule, length=8.0865, radius=10.7894, shapeName=pelvis_capsule, center=<1.9773 0.5889 0.0059>, rotation=<rot 5.1404 -90 -0.0505>}]
 spine_01 [{
   shapeType=Capsule
   length=3.9629
   radius=14.6491
   shapeName=spine_01_capsule
   center=<8.3036 0.306 0.0704>
   rotation=<rot -90 -0.0004 -0.0002>
  }]
 thigh_l [{
   shapeType=Capsule
   length=24.5493
   radius=9.3854
   shapeName=thigh_l_capsule
   center=<-12.3055 -3.0069 1.1021>
   rotation=<rot 14.6106 -90.0001 -88.5559>
  }]
 thigh_r [
  {
   shapeType=Capsule
   length=24.5493
   radius=9.3854
   shapeName=thigh_l_capsule
   center=<12.3055 2.5132 -0.9734>
   rotation=<rot 14.6106 -90.0001 91.4441>
  }
  {
   shapeType=Capsule
   length=11.0549
   radius=5.9637
   shapeName=thigh_l_capsule
   center=<7.4534 -2.576 0.2797>
   rotation=<rot 88.8874 2.002 -166.2228>
  }
 ]
}
constraints [
 {
  bone1Name=thigh_l
  bone2Name=pelvis
  swing1LimitDegrees=45
  swing1Motion=Locked
  swing2LimitDegrees=45
  swing2Motion=Locked
  twistLimitDegrees=45
  twistMotion=Locked
 }
 {
  bone1Name=spine_01
  bone2Name=pelvis
  swing1LimitDegrees=45
  swing1Motion=Free
  swing2LimitDegrees=45
  swing2Motion=Free
  twistLimitDegrees=45
  twistMotion=Free
 }
 {bone1Name=belt_l, bone2Name=pelvis, swing1LimitDegrees=25, swing2LimitDegrees=10, twistLimitDegrees=5}
 {bone1Name=belt_l1, bone2Name=belt_l, swing1LimitDegrees=15, swing2LimitDegrees=10, twistLimitDegrees=5}
 {bone1Name=belt_l2, bone2Name=belt_l1, swing1LimitDegrees=15, swing2LimitDegrees=10, twistLimitDegrees=5}
 {bone1Name=belt_r, bone2Name=pelvis, swing1LimitDegrees=15, swing2LimitDegrees=5, twistLimitDegrees=2}
 {bone1Name=belt_r1, bone2Name=belt_r, swing1LimitDegrees=20, swing2LimitDegrees=12, twistLimitDegrees=2}
 {bone1Name=belt_r2, bone2Name=belt_r1, swing1LimitDegrees=25, swing2LimitDegrees=10, twistLimitDegrees=2}
 {bone1Name=Belts_l_jnt_01, bone2Name=spine_01, swing1LimitDegrees=15, swing2LimitDegrees=20, twistLimitDegrees=45, twistMotion=Locked}
 {bone1Name=Belts_r_jnt_01, bone2Name=spine_01, swing1LimitDegrees=15, swing2LimitDegrees=20, twistLimitDegrees=45, twistMotion=Locked}
 {
  bone1Name=thigh_r
  bone2Name=pelvis
  swing1LimitDegrees=45
  swing1Motion=Locked
  swing2LimitDegrees=45
  swing2Motion=Locked
  twistLimitDegrees=45
  twistMotion=Locked
 }
]
