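# Secondary-motion AnimBP for the male medium SpeedDialBattle body cosmetic, layering necklace and belt dynamics over a pose copied from the attached parent mesh.
# A crouch-enabled Control Rig and belt constraint feed necklace RigidBody simulation, then a bool blend of two physics branches with additive belt-bone offsets.
# Spine and pelvis constraints anchor the belt joints before a deform Control Rig, which adds extra deformation while targeting.
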AnimBlueprint /BRCosmetics/Characters/Player/Male/Medium/Bodies/M_MED_SpeedDialBattle/Meshes/M_MED_SpeedDialBattle_AnimBP
target_skeleton=&/M_MED_SpeedDialBattle_Skeleton
graphs [
 AnimGraph {nodes [
   RigidBody!(ComponentPose, Alpha, ExternalForce, ComponentLinearAccScale:Vector="0.2,0.15,0.3", bTransferBoneVelocities:bool=False) -> (Pose) {
    bOverrideWorldGravity=true
    overridePhysicsAsset=&/M_MED_SpeedDialBattle_Belts_Physics
    alphaBoolBlend {blendOption=Linear}
    baseBoneRef {boneName=pelvis}
    componentAppliedLinearAccClamp=<1000 1000 1000>
    componentLinearAccScale=<0.2 0.15 0.3>
    componentLinearVelScale=<0.1 0.1 0.2>
    overrideWorldGravity {z=-300}
    simSpaceSettings {
     maxAngularAcceleration=1000
     maxAngularVelocity=3
     maxLinearAcceleration=1000
     maxLinearVelocity=20
     velocityScaleZ=0.5
     worldAlpha=1
     externalAngularVelocity {z=2}
    }
   }
   n1 = Control Rig!(Source=Use cached pose 'results_animation_cache'(), Alpha=0.65, Enable_extra_deform:bool=Get Is Targeting() -> (Is Targeting:bool)) -> (Pose) {
    alphaBoolBlend {blendOption=Linear}
    alphaScaleBiasClamp {interpSpeedDecreasing=15, interpSpeedIncreasing=15}
    controlRigAssetReference {blueprintRigClass=&/M_MED_SpeedDialBattle_DeformRig.M_MED_SpeedDialBattle_DeformRig_C}
   }
   Output Pose(Result=@n1) {layerGroup=None}
   n2 = Control Rig!(Source=Copy Pose From Mesh() {bUseAttachedParent=true}, bAlphaBoolEnabled:bool=Get bIsCrouching()) -> (Pose) {
    alphaInputType=Bool
    alphaBoolBlend {blendInTime=0.1, blendOption=Linear, blendOutTime=0.1}
    controlRigAssetReference {blueprintRigClass=&/M_MED_SpeedDialBattle_CtrlRig.M_MED_SpeedDialBattle_CtrlRig_C}
   }
   n3 = "Constraint - belt_r (RR)"!(ComponentPose=Local To Component(LocalPose=@n2), Alpha=0.3, ConstraintWeights_0=0.5, ConstraintWeights_1:float=0.5) -> (Pose) {
    alphaBoolBlend {blendOption=Linear}
    boneToModify {boneName=belt_r}
    constraintSetup [{transformType=Rotation, targetBone {boneName=thigh_r}}, {transformType=Rotation, targetBone {boneName=pelvis}}]
    constraintWeights [1,1]
   }
   Save cached pose 'pre_Dyn'(Pose=Component To Local(ComponentPose=@n3)) {cacheName=pre_Dyn}
   n4 = RigidBody(
    ComponentPose=Local To Component(LocalPose=Use cached pose 'Add_necklace_dyn'())
    ComponentLinearAccScale="0.2,0.15,0.3"
    bTransferBoneVelocities=False
   ) {
    bOverrideWorldGravity=true
    overridePhysicsAsset=&M_MED_SpeedDialBattle_Belts_Physics
    simulationSpace=BaseBoneSpace
    alphaBoolBlend {blendOption=Linear}
    baseBoneRef {boneName=pelvis}
    componentAppliedLinearAccClamp=<1000 1000 1000>
    componentLinearAccScale=<0.2 0.15 0.3>
    componentLinearVelScale=<0.1 0.1 0.2>
    overrideWorldGravity {z=-300}
    simSpaceSettings {
     maxAngularAcceleration=1000
     maxAngularVelocity=150
     maxLinearAcceleration=1000
     maxLinearVelocity=20
     worldAlpha=1
     externalAngularVelocity {z=2}
    }
   }
   n5 = "Transform (Modify) Bone - Bone: belt_r"!(ComponentPose=@n4, bAlphaBoolEnabled:bool=True, Translation="0,0,0", Rotation="0.020,1.210,0.002", Scale="1,1,1") -> (Pose) {
    alphaInputType=Bool
    rotationMode=BMM_Additive
    rotationSpace=BCS_BoneSpace
    translationMode=BMM_Additive
    translationSpace=BCS_BoneSpace
    alphaBoolBlend {blendInTime=0.3, blendOption=Linear, blendOutTime=0.1}
    boneToModify {boneName=belt_r}
    rotation=<rot 0.02 1.21 0.002>
   }
   n6 = "Transform (Modify) Bone - Bone: belt_r1"!(ComponentPose=@n5, bAlphaBoolEnabled:bool=True, Translation="0,0,0", Rotation="8.822,-1.792,-2.981", Scale="1,1,1") -> (Pose) {
    alphaInputType=Bool
    rotationMode=BMM_Additive
    rotationSpace=BCS_BoneSpace
    translationMode=BMM_Additive
    translationSpace=BCS_BoneSpace
    alphaBoolBlend {blendInTime=0.1, blendOption=Linear, blendOutTime=0.1}
    boneToModify {boneName=belt_r1}
    rotation=<rot 8.822 -1.7919 -2.981>
   }
   n7 = "Transform (Modify) Bone - Bone: belt_l"!(ComponentPose=@n6, bAlphaBoolEnabled:bool=True, Translation="0,0,0", Rotation="-3.274,-1.672,-10.744", Scale="1,1,1") -> (Pose) {
    alphaInputType=Bool
    rotationMode=BMM_Additive
    rotationSpace=BCS_BoneSpace
    translationMode=BMM_Additive
    translationSpace=BCS_BoneSpace
    alphaBoolBlend {blendInTime=0.1, blendOption=Linear, blendOutTime=0.3}
    boneToModify {boneName=belt_l}
    rotation=<rot -3.274 -1.672 -10.744>
   }
   n8 = RigidBody(
    ComponentPose=Local To Component(LocalPose=Use cached pose 'Add_necklace_dyn'())
    ComponentLinearAccScale="0.2,0.15,0.3"
    bTransferBoneVelocities=False
   ) {
    bOverrideWorldGravity=true
    bTransferBoneVelocities=true
    overridePhysicsAsset=&/M_MED_SpeedDialBattle_Physics
    simulationSpace=BaseBoneSpace
    alphaBoolBlend {blendOption=Linear}
    baseBoneRef {boneName=pelvis}
    componentAppliedLinearAccClamp=<1000 1000 1000>
    componentLinearAccScale=<0.2 0.15 0.3>
    componentLinearVelScale=<0.3 0.3 0.3>
    overrideWorldGravity {z=-450}
    simSpaceSettings {maxAngularAcceleration=1500, maxAngularVelocity=350, maxLinearAcceleration=1500, maxLinearVelocity=30, worldAlpha=1}
   }
   n9 = "Transform (Modify) Bone - Bone: belt_r"!(ComponentPose=@n8, Alpha:float=1, Translation="0,0,0", Rotation="0,0,0", Scale="1,1,1") -> (Pose) {
    rotationMode=BMM_Additive
    rotationSpace=BCS_BoneSpace
    translationMode=BMM_Additive
    translationSpace=BCS_BoneSpace
    alphaBoolBlend {blendOption=Linear}
    alphaScaleBiasClamp {bInterpResult=true}
    boneToModify {boneName=belt_r}
   }
   n10 = "Transform (Modify) Bone - Bone: belt_r1"!(ComponentPose=@n9, Alpha:float=1, Translation="0,0,0", Rotation="0,0,0", Scale="1,1,1") -> (Pose) {
    rotationMode=BMM_Additive
    rotationSpace=BCS_BoneSpace
    translationMode=BMM_Additive
    translationSpace=BCS_BoneSpace
    alphaBoolBlend {blendOption=Linear}
    alphaScaleBiasClamp {bInterpResult=true}
    boneToModify {boneName=belt_r1}
   }
   Save cached pose 'dyn_cache'(Pose=Blend Poses by bool(BlendPose_0=Component To Local(ComponentPose=@n10), BlendPose_1=Component To Local(ComponentPose=@n7), BlendTime_0=0.3, BlendTime_1=0.5) {blendType=Linear}) {cacheName=dyn_cache}
   n11 = "Constraint - Belts_r_offset_jnt (PPPP)"!(
    ComponentPose=Local To Component(LocalPose=Use cached pose 'dyn_cache'())
    Alpha=1
    ConstraintWeights_0=0.3295
    ConstraintWeights_1:float=0.2421
    ConstraintWeights_2:float=0.2177
    ConstraintWeights_3:float=0.2107
   ) -> (Pose) {
    alphaBoolBlend {blendOption=Linear}
    boneToModify {boneName=Belts_r_offset_jnt}
    constraintSetup [{targetBone {boneName=spine_02}}, {targetBone {boneName=pelvis}}, {targetBone {boneName=spine_01}}, {targetBone {boneName=spine_03}}]
    constraintWeights [1,1,1,1]
   }
   n12 = "Constraint - Belts_r_jnt_01 (PPPP)"!(
    ComponentPose=@n11
    Alpha=0.45
    ConstraintWeights_0=0.3295
    ConstraintWeights_1:float=0.2421
    ConstraintWeights_2:float=0.2177
    ConstraintWeights_3:float=0.2107
   ) -> (Pose) {
    alphaBoolBlend {blendOption=Linear}
    boneToModify {boneName=Belts_r_jnt_01}
    constraintSetup [{targetBone {boneName=spine_02}}, {targetBone {boneName=pelvis}}, {targetBone {boneName=spine_01}}, {targetBone {boneName=spine_03}}]
    constraintWeights [1,1,1,1]
   }
   n13 = "Constraint - Belts_l_offset_jnt (PPPP)"!(
    ComponentPose=@n12
    Alpha=1
    ConstraintWeights_0=0.3295
    ConstraintWeights_1:float=0.2421
    ConstraintWeights_2:float=0.2177
    ConstraintWeights_3:float=0.2107
   ) -> (Pose) {
    alphaBoolBlend {blendOption=Linear}
    boneToModify {boneName=Belts_l_offset_jnt}
    constraintSetup [{targetBone {boneName=spine_02}}, {targetBone {boneName=pelvis}}, {targetBone {boneName=spine_01}}, {targetBone {boneName=spine_03}}]
    constraintWeights [1,1,1,1]
   }
   n14 = "Constraint - Belts_l_jnt_01 (PPPP)"!(
    ComponentPose=@n13
    Alpha=0.45
    ConstraintWeights_0=0.3295
    ConstraintWeights_1:float=0.2421
    ConstraintWeights_2:float=0.2177
    ConstraintWeights_3:float=0.2107
   ) -> (Pose) {
    alphaBoolBlend {blendOption=Linear}
    boneToModify {boneName=Belts_l_jnt_01}
    constraintSetup [{targetBone {boneName=spine_02}}, {targetBone {boneName=pelvis}}, {targetBone {boneName=spine_01}}, {targetBone {boneName=spine_03}}]
    constraintWeights [1,1,1,1]
   }
   Save cached pose 'results_animation_cache'(Pose=Component To Local(ComponentPose=@n14)) {cacheName=results_animation_cache}
   n15 = RigidBody!(ComponentPose=Local To Component(LocalPose=Use cached pose 'pre_Dyn'()), Alpha, ExternalForce, ComponentLinearAccScale="0.2,0.15,0.3") -> (Pose) {
    bOverrideWorldGravity=true
    overridePhysicsAsset=&/M_MED_SpeedDialBattle_Necklace_Physics
    simulationSpace=BaseBoneSpace
    alphaBoolBlend {blendOption=Linear}
    baseBoneRef {boneName=root}
    componentAppliedLinearAccClamp=<1000 1000 1000>
    componentLinearAccScale=<0.2 0.15 0.3>
    componentLinearVelScale=<0.3 0.3 0.3>
    overrideWorldGravity {z=-650}
    simSpaceSettings {maxAngularAcceleration=1500, maxAngularVelocity=15, maxLinearAcceleration=1500, maxLinearVelocity=70}
   }
   Save cached pose 'Add_necklace_dyn'(Pose=Component To Local(ComponentPose=@n15)) {cacheName=Add_necklace_dyn}
  ]}
 EventGraph {nodes [
   Event BlueprintUpdateAnimation()
   > n1 = Sequence!(execute) -> (then_0, then_1, then_2:Exec)
   n2 = Get bIsMoving2D()
   n3 = Get bIsCrouching()
   > Set cond_shaking_balance(cond_shaking_balance:bool=OR Boolean(A=AND Boolean(A=@n2, B=@n3), B=Get bIsJumping())) -> (Output_Get:bool)
   Set Is Targeting(execute=@n1.then_1, Is Targeting:bool=Get bIsTargeting()) -> (Output_Get:bool)
   Set belt_correction_cond(execute=@n1.then_2, belt_correction_cond:bool=AND Boolean(A="Not Equal (Boolean)"(A=@n2, B=true) -> (ReturnValue=~), B=@n3)) -> (Output_Get:bool)
  ]}
]
variables [
 belt_correction_cond {type=boolean}
 cond_shaking_balance {type=boolean}
 Is Targeting {type=boolean}
 NewVar {type=boolean}
 Pawn Velocity Z {type=number}
]
